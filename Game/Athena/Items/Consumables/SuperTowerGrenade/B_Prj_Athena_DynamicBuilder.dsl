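; Projectile Blueprint for Athena dynamic-builder grenade consumables that spawns a prefab structure in the world.
; CallDynamicBuilder spawns the given builder class at a location, with yaw looked up from the cardinal direction, then initializes it.
; When the builder finishes spawning, a bound event marks every building as player-placed for destruction.
; RevertToPickup tosses the item back as a pickup if valid, then hides and destroys the projectile.
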(event Custom|BPDynamicBuilderFinishedSpawning_Event
  (params (Buildings :type Array%20of%20Building%20Actor%20Object%20References))
  (for _array_element Buildings
    (BuildingActor|SetPlayerPlacedforDestruction _array_element true)))

(event Custom|CallDynamicBuilder
  (params (Class :type Actor%20Class%20Reference) (Location :type Vector))
  (bind _asbuilding_actor (Utilities|Casting|CastToBuildingActorClass Class))
  (bind _returnvalue (Building|SpawnBuildingActor _asbuilding_actor (Math|Transform|MakeTransform Location (Math|Rotator|MakeRotator 0.0 0.0 (Utilities|Map|Find (Variables|Orientation|GetOrientationYawOffsets) (Variables|Orientation|GetCardinalDirection)))) (Game|GetInstigator) (Game|GetInstigator) true))
  (bind _asfort_dynamic_builder (Utilities|Casting|CastToFortDynamicBuilder _returnvalue)
    (:then
      (EventDispatchers|AssignBPDynamicBuilderFinishedSpawning _asfort_dynamic_builder (delegate Custom|BPDynamicBuilderFinishedSpawning_Event)))
    (:CastFailed
      _))
  (Building|InitializeKismetSpawnedBuildingActor _asfort_dynamic_builder 0 0 true 0 true))

(fn PostInitializedBuildingActor
  (params (BuiltActor :type Building%20Actor%20Object%20Reference)))

(fn RevertToPickup
  (params (WorldItemDef :type Fort%20World%20Item%20Definition%20Object%20Reference))
  (Utilities|IsValid WorldItemDef
    (:Is%20Valid
      (Inventory|SpawnPickupInWorld WorldItemDef 1 (Transformation|GetActorLocation) "0.000000,0.000000,1.000000" 0 true false false -1 "Tossed" "TossedByPlayer" 0 false))
    (:Is%20Not%20Valid
      _))
  (Rendering|SetActorHiddenInGame true)
  (Actor|DestroyActor))

(fn SelectActor
  (params (A :type Actor%20Object%20Reference) (B :type Actor%20Object%20Reference) (Choose%20A :type Boolean))
  (if Choose%20A
    (return A)
    (else
      (return B))))

(fn ConstructionScript)
