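; Super Tower Grenade projectile that bounces off pawns, vehicles and similar actors, then explodes into a grid-snapped tower facing the thrower's cardinal direction.
; On stop it snaps the impact, preferring client-predicted location within tolerance.
; Non-cosmetic explosion logic box-traces the doorway, choosing normal or raised basement builder, with optional grid clearing and tires.
; Props queued for removal are destroyed one per timer tick, spreading the cost.
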(event EventBeginPlay
  ;[1 1 1] Lookup build delay row. Determine which cardinal direction the tower should face.
  (bind _returnvalue (Audio|SpawnSoundAttached :Sound (Variables|Cosmetic|GetInAirLoopSound) :AttachToComponent (Variables|Collision|GetCapsuleComponent) :AttachPointName "None" :LocationType "KeepRelativeOffset" :bStopWhenAttachedToDestroyed true :VolumeMultiplier 1.0 :PitchMultiplier 1.0 :StartTime 0.0 :bAutoDestroy true))
  (bind _output_get (Variables|Components|SetInAirLoopComponent _returnvalue))
  (Audio|Components|Audio|FadeIn _output_get 1.0 1.0 0.0 "Linear" :!class /Script/Engine.AudioComponent)
  (CallFunction|CalculateCardinalDirection))

(event EventEndPlay
  (params (EndPlayReason :type EEndPlayReason%20Enum :default "Destroyed"))
  (bind _buildingstructuralsupportsystem (Building|GetBuildingStructuralSupportSystem))
  ;[1 1 1] Destroy any remaining actors in our destroy list if needed
  (switch Utilities|FlowControl|Switch|SwitchonEEndPlayReason EndPlayReason
    (:Destroyed
      (Building|StartActorRemovalBatch _buildingstructuralsupportsystem)
      (for _array_element (Variables|Tower|GetBuildingActorsToDestroy)
        (Actor|DestroyActor _array_element))
      (Building|StopActorRemovalBatch _buildingstructuralsupportsystem))))

(event Collision|EventHit
  (params (MyComp :type Primitive%20Component%20Object%20Reference) (Other :type Actor%20Object%20Reference) (OtherComp :type Primitive%20Component%20Object%20Reference) (bSelfMoved :type Boolean) (HitLocation :type Vector :default "0, 0, 0") (HitNormal :type Vector :default "0, 0, 0") (NormalImpulse :type Vector :default "0, 0, 0") (Hit :type Hit%20Result%20Structure%20%28by%20ref%29))
  ;[1 1 1] Bounce off of pawns and walls
  (bind _returnvalue (CallFunction|ShouldBounceOffActor Other OtherComp HitNormal))
  (if _returnvalue
    (CallFunction|HandleBounce HitLocation)
    (Utilities|FlowControl|DoOnce false
      (:Completed
        (Utilities|FlowControl|Delay (Variables|Cosmetic|GetBounceTimeout)))))
  (Utilities|FlowControl|DoOnce false
    (:Completed
      (Game|Components|ProjectileMovement|StopSimulating (Variables|Movement|GetProjectileMovementComponent) Hit))))

(event Custom|Handle%20Bounce
  (params (Hit%20Location :type Vector))
  (bind _lastbouncelocation (Variables|Cosmetic|GetLastBounceLocation))
  (bind _bouncesound (Variables|Cosmetic|GetBounceSound))
  ;[1 1 1] Play audio feedback, etc when bouncing
  (if (and (> (Math|Vector|VectorLengthSquared (- Hit%20Location (Variables|Cosmetic|GetLastBounceLocation))) (Math|Float|Square (Variables|Cosmetic|GetMinBounceDistance))) (> (- (Utilities|Time|GetTimeSeconds) (Variables|Cosmetic|GetLastBounceTime)) (Variables|Cosmetic|GetMinTimeBetweenBounces)))
    (Variables|Cosmetic|SetLastBounceLocation Hit%20Location)
    (Variables|Cosmetic|SetLastBounceTime (Utilities|Time|GetTimeSeconds))
    (Audio|PlaySoundatLocation _bouncesound _lastbouncelocation)
    (Fortnite|ShowSoundIndicatorLocalforTeam :Location _lastbouncelocation :MaxAudibleDistance (Variables|/Script/Engine.SoundBase|GetMaxDistance _bouncesound) :IndicatorType "Gunshot" :Team (Team|GetActorTeam self) :Affiliations (Utilities|Array|MakeArray "Hostile" "Neutral") :Instigator self :Tint "(R=0.000000,G=0.000000,B=0.000000,A=0.000000)" :AttenuationMaxRangeMultiplier 1.0 :AttenuationIntensityMultiplier 1.0)))

(event EventOnStop
  (params (Hit :type Hit%20Result%20Structure%20%28by%20ref%29))
  (bind (_bblockinghit _binitialoverlap _time _distance _location _impactpoint _normal _impactnormal _physmat _hitactor _hitcomponent) (Collision|BreakHitResult Hit))
  ;[1 1 1] Initializing variables, stop rotation, and snap projectile to grid before "exploding"
  (Rendering|SetVisibility (Variables|Components|GetMesh) false false)
  (CallFunction|SetBaseDestination Hit)
  (Components|Movement|SetUpdatedComponent (Variables|Components|GetRotatingMovement) 0)
  (Variables|Default|SetSnapToGridLocation (Fortnite|SnaptoFortBuildingGrid (Math|Vector|SelectVector (Variables|/Game/Athena/Items/Consumables/Parents/B_Prj_Comp_ClientPredicted_Data.B_Prj_Comp_ClientPredicted_Data_C|GetClientPredictedImpactLocation (Variables|Default|GetB_Prj_Comp_ClientPredicted_Data)) (Transformation|GetActorLocation) (Utilities|Operators|Equal(==) (Variables|/Game/Athena/Items/Consumables/Parents/B_Prj_Comp_ClientPredicted_Data.B_Prj_Comp_ClientPredicted_Data_C|GetClientPredictedImpactLocation (Variables|Default|GetB_Prj_Comp_ClientPredicted_Data)) (Transformation|GetActorLocation) (Variables|Default|GetClientLocationErrorTolerance))) (Math|Vector|MakeVector 0.0 0.0 (Variables|SnaptoGrid|GetGridZOffset)) (Math|Vector|MakeVector 0.0 (Variables|SnaptoGrid|GetGridYOffset))))
  (Transformation|SetActorTransform self (Math|Transform|MakeTransform (Transformation|GetActorLocation) (Math|Rotator|MakeRotator 0.0 0.0 (Utilities|Map|Find (Variables|Orientation|GetOrientationYawOffsets) (Variables|Orientation|GetCardinalDirection)))) false false)
  (bind _returnvalue (CallFunction|ShouldBounceOffActor _hitactor _hitcomponent _normal)) ; Same conditions to check for if the building should spawn
  (if _returnvalue
    (bind _reverttopickupwid (Variables|Default|GetRevertToPickupWID)
      (:then
        (CallFunction|ReverttoPickup _reverttopickupwid))
      (:else
        _)))
  (Delivery|DoExplosion))

(event EventOnExploded
  (params (HitActors :type Array%20of%20Actor%20Object%20References) (HitResults :type Array%20of%20Hit%20Result%20Structures))
  (bind _gridyoffset (Variables|SnaptoGrid|GetGridYOffset))
  (bind _gridzoffset (Variables|SnaptoGrid|GetGridZOffset))
  (bind _returnvalue (Transformation|GetActorLocation))
  (bind _returnvalue_2 (Fortnite|SnaptoFortBuildingGrid _returnvalue (Math|Vector|MakeVector 0.0 0.0 _gridzoffset) (Math|Vector|MakeVector 0.0 _gridyoffset)))
  (bind _returnvalue_3 (Game|GetInstigator))
  (bind _returnvalue_4 (Pawn|IsLocallyControlled _returnvalue_3))
  (bind _cardinaldirection (Variables|Orientation|GetCardinalDirection))
  (bind _orientationyawoffsets (Variables|Orientation|GetOrientationYawOffsets))
  (bind _value (Utilities|Map|Find _orientationyawoffsets _cardinaldirection))
  (bind _cue_closesound (Variables|Cosmetic|GetCue_CloseSound))
  (bind _cue_distantsound (Variables|Cosmetic|GetCue_DistantSound))
  (bind _returnvalue_5 (select _returnvalue_4 _cue_closesound _cue_distantsound))
  (bind _snaptogridlocation (Variables|Default|GetSnapToGridLocation))
  (Networking|CanExecuteCosmeticEvents
    (:True
      ;[1 1 1] OnExploded - Emitter, Sound
      (CallFunction|SpawnBuildFX _snaptogridlocation)
      (Camera|PlayWorldCameraShake "/Game/Athena/Items/Gameplay/BackPacks/CarminePack/FX/B_Ground_CameraShake_Heavy.B_Ground_CameraShake_Heavy_C" _snaptogridlocation 8e+02 4.5e+03 1.8)
      (Fortnite|PlayLocalForceFeedbackatLocationMulti "/Game/Weapons/ForceFeedback/Explosions/FF_Generic_Explosion_Near.FF_Generic_Explosion_Near" "/Game/Weapons/ForceFeedback/Explosions/FF_Generic_Explosion_Far.FF_Generic_Explosion_Far" _snaptogridlocation 1e+03 2.5e+03)
      (Audio|Components|Audio|FadeOut (Variables|Components|GetInAirLoopComponent) 0.2 0.0 "Linear" :!class /Script/Engine.AudioComponent)
      (Audio|PlaySoundatLocation _returnvalue_5 _returnvalue_2) ; Play Close Explosion
      (Fortnite|ShowSoundIndicatorLocalforTeam :Location _returnvalue_2 :MaxAudibleDistance (Variables|/Script/Engine.SoundBase|GetMaxDistance _returnvalue_5) :IndicatorType "Gunshot" :Team (Team|GetActorTeam self) :Affiliations (Utilities|Array|MakeArray "Hostile" "Neutral") :Instigator self :Tint "(R=0.000000,G=0.000000,B=0.000000,A=0.000000)" :AttenuationMaxRangeMultiplier 1.0 :AttenuationIntensityMultiplier 1.0))
    (:False
      ;[1 1 1] Trace to see if a player pawn can walk through the door.
      (CallFunction|HandleProps HitActors)
      (bind _outhits (Collision|MultiBoxTraceForObjects (Math|Vector|vector+vector (Variables|BuildingAndClearing|GetBaseDestination) (Variables|TerrainTrace|GetDoorTraceOffsetHigh) (Utilities|Map|Find (Variables|Orientation|GetOrientationDoortraceOffsets) (Variables|Orientation|GetCardinalDirection))) (Math|Vector|vector+vector (Variables|BuildingAndClearing|GetBaseDestination) (Variables|TerrainTrace|GetDoorTraceOffsetLow) (Utilities|Map|Find (Variables|Orientation|GetOrientationDoortraceOffsets) (Variables|Orientation|GetCardinalDirection))) (Variables|Default|GetDoorCheckSize) "0, 0, 0" (Utilities|Array|MakeArray "ObjectTypeQuery1" "ObjectTypeQuery2") false (Variables|Tower|GetBuildingActorsToIgnore)))
      ;[1 1 1] Check if any hits are against terrain instead of building actors
      (for _array_element _outhits
        (Development|PrintString (Utilities|String|Append "Tower Grenade Hit: " (Utilities|GetDisplayName (.HitActor (Collision|BreakHitResult _array_element)))) (Variables|Default|GetIsDebug) (Variables|Default|GetIsDebug) "(R=0.000000,G=0.660000,B=1.000000,A=1.000000)" 1e+01)
        (if (Math|Boolean|ORBoolean (Utilities|IsA (.HitActor (Collision|BreakHitResult _array_element)) "/Script/Landscape.Landscape") (Utilities|IsA (.HitActor (Collision|BreakHitResult _array_element)) "/Script/Landscape.LandscapeStreamingProxy") (Utilities|IsA (.HitActor (Collision|BreakHitResult _array_element)) "/Script/FortniteGame.BuildingStaticGeo") (Utilities|IsA (.HitActor (Collision|BreakHitResult _array_element)) "/Script/FortniteGame.FortStaticMeshActor"))
          (Variables|TerrainTrace|Setb_HitTerrain true)
          (else
            (bind _asbuilding_actor (Utilities|Casting|CastToBuildingActor (.HitActor (Collision|BreakHitResult _array_element))))
            (Default|IsNotAVolume _asbuilding_actor (.HitComponent (Collision|BreakHitResult _array_element))
              (:IsNotVolume
                (Variables|TerrainTrace|Setb_HitTerrain (Math|Boolean|ORBoolean (not (Variables|Actor|GetbCanBeDamaged _asbuilding_actor)) (Variables|/Script/FortniteGame.BuildingActor|GetbStructurallySupportOverlappingActors _asbuilding_actor) (Variables|TerrainTrace|Getb_HitTerrain))))))))
      (if (Variables|TerrainTrace|Getb_HitTerrain)
        ;[1 1 1] Door would be underground, so it is moved up a grid cell and we add stairs and archways to the bottom of the tower.
        (if (Variables|Default|GetbClearGridCells)
          (Tower|ClearGridCells (Variables|Default|GetClearedCells_Underground) true (Variables|BuildingAndClearing|GetBaseDestination) _value))
        (CallFunction|CallDynamicBuilder (Variables|Default|GetBuilderClassBasement) (Variables|BuildingAndClearing|GetBaseDestination))
        (Variables|BuildingAndClearing|SetBaseDestination (+ (Variables|BuildingAndClearing|GetBaseDestination) "0, 0,384.000000"))
        (elif (Variables|Default|GetbClearGridCells)
          (Tower|ClearGridCells (Variables|Default|GetClearedCells_Normal) true (Variables|BuildingAndClearing|GetBaseDestination) _value)
          (CallFunction|CallDynamicBuilder (Variables|Default|GetBuilderClass) (Variables|BuildingAndClearing|GetBaseDestination) :!label _label)
          (else
            (goto _label))))
      (if (Variables|Default|Getb_SpawnTires)
        (Utilities|FlowControl|Delay (Variables|Default|GetTireBuildDelay))
        (CallFunction|SpawnTires (Variables|BuildingAndClearing|GetBaseDestination)))
      ;[1 1 1] loop a timer deleting an actor every frame or so
      (bind _returnvalue_6 (Utilities|Time|SetTimerbyEvent (delegate Custom|Delete%20another%20actor) 0.03 true)))))

(event Custom|Delete%20another%20actor
  (bind _buildingactorstodestroy (Variables|Tower|GetBuildingActorsToDestroy))
  ;[1 1 1] loop a timer deleting an actor every frame or so
  (if (> (Utilities|Array|Length _buildingactorstodestroy) 0)
    (bind (_array_element _array_index) (Utilities|Array|ForEachLoopwithBreak _buildingactorstodestroy :!label _label)
      (:LoopBody
        (Utilities|IsValid _array_element
          (:Is%20Valid
            (if (Variables|Actor|GetbCanBeDamaged _array_element)
              (Actor|DestroyActor _array_element)))
          (:Is%20Not%20Valid
            _))
        (Utilities|Array|RemoveIndex _buildingactorstodestroy _array_index)
        (goto _label Break)))
    (else
      (Utilities|Time|ClearandInvalidateTimerbyHandle (stale EventOnExploded _returnvalue_6)))))

(fn CalculateCardinalDirection
  (bind _returnvalue (Game|GetInstigator))
  (bind _returnvalue_2 (Transformation|GetActorForwardVector _returnvalue))
  (bind _x (.x _returnvalue_2))
  (bind _y (.y _returnvalue_2))
  (if (>= (Math|Float|Absolute(Float) _x) (Math|Float|Absolute(Float) _y))
    (if (>= _x 0.0)
      (Variables|Orientation|SetCardinalDirection "NewEnumerator0")
      (else
        (Variables|Orientation|SetCardinalDirection "NewEnumerator1")))
    (elif (>= _y 0.0)
      (Variables|Orientation|SetCardinalDirection "NewEnumerator2")
      (else
        (Variables|Orientation|SetCardinalDirection "NewEnumerator3")))))

(fn CanDestroyObject
  (params (Object :type Actor%20Object%20Reference))
  (bind _object (Variables|GetObject))
  (bind _bcanbedamaged (Variables|Actor|GetbCanBeDamaged _object))
  (bind _asbuilding_actor (Utilities|Casting|CastToBuildingActor Object)
    (:then
      (LocalVariables|SetBuilding _asbuilding_actor)
      (bind _returnvalue (CallFunction|GetOwnerWeapon))
      (Utilities|IsValid _returnvalue
        (:Is%20Valid
          (Utilities|IsValid (Actor|GetOwner _returnvalue)
            (:Is%20Valid
              (bind _aspawn (Utilities|Casting|CastToPawn (Actor|GetOwner _returnvalue))
                (:then
                  (bind _returnvalue_2 (Class|BuildingActor|CanDestroyBuildingActor (LocalVariables|GetBuilding) (Pawn|GetController _aspawn) _aspawn))
                  (return _returnvalue_2))
                (:CastFailed
                  (return _bcanbedamaged))))
            (:Is%20Not%20Valid
              (return _bcanbedamaged))))
        (:Is%20Not%20Valid
          (return _bcanbedamaged))))
    (:CastFailed
      (return (Variables|Actor|GetbCanBeDamaged (Variables|GetObject))))))

(fn CheckForBlockedAreas)

(fn ClearAndBuild
  (params (ClearedGridCells :type Array%20of%20Vectors) (BGAClassArray :type Array%20of%20Actor%20Class%20References) (BGALocationArray :type Array%20of%20Vectors) (BGARotationArray :type Array%20of%20Rotators) (DestroyFloorsQuickly :type Boolean))
  (for _array_element ClearedGridCells
    (Tower|DestroyActorsinGridCellLocation (Building|GetBuildingStructuralSupportSystem) (+ (Variables|BuildingAndClearing|GetBaseDestination) (Math|Vector|RotateVectorAroundAxis _array_element (Utilities|Map|Find (Variables|Orientation|GetOrientationYawOffsets) (Variables|Orientation|GetCardinalDirection)) "0, 0,1.000000")) DestroyFloorsQuickly))
  (CallFunction|CreateBaseSection (Variables|BuildingAndClearing|GetBaseDestination) BGAClassArray BGALocationArray BGARotationArray))

(fn ClearOnly
  (params (ClearedGridCells :type Array%20of%20Vectors) (DestroyFloorsQuickly :type Boolean))
  (bind _buildingstructuralsupportsystem (Building|GetBuildingStructuralSupportSystem))
  (Building|StartActorRemovalBatch _buildingstructuralsupportsystem)
  (for _array_element ClearedGridCells
    (Tower|DestroyActorsinGridCellLocation (Building|GetBuildingStructuralSupportSystem) (+ (Variables|BuildingAndClearing|GetBaseDestination) (Math|Vector|RotateVectorAroundAxis _array_element (Utilities|Map|Find (Variables|Orientation|GetOrientationYawOffsets) (Variables|Orientation|GetCardinalDirection)) "0, 0,1.000000")) DestroyFloorsQuickly))
  (Building|StopActorRemovalBatch _buildingstructuralsupportsystem))

(fn CreateBaseSection
  (params (ReferenceLocation :type Vector) (BGAClassArray :type Array%20of%20Actor%20Class%20References) (BGALocationArray :type Array%20of%20Vectors) (BGARotationArray :type Array%20of%20Rotators))
  (if (> (Utilities|Array|Length BGAClassArray) 0)
    (if (Creative|Budget|CanPlaceClassInstances (Fortnite|GetFortPlayerControllerfromActor (Actor|GetOwner)) BGAClassArray ReferenceLocation)
      (for _index (range (Utilities|Array|Length BGAClassArray))
        (bind _returnvalue (Game|SpawnBuildingActorfromClass (Utilities|Array|Get(acopy) BGAClassArray _index) (Math|Transform|MakeTransform (Math|Vector|vector+vector ReferenceLocation (Math|Vector|RotateVectorAroundAxis (Utilities|Array|Get(aref) BGALocationArray _index) (Utilities|Map|Find (Variables|Orientation|GetOrientationYawOffsets) (Variables|Orientation|GetCardinalDirection)) "0, 0,1.000000") (Utilities|Map|Find (Variables|Orientation|GetOrientationLocationOffsets) (Variables|Orientation|GetCardinalDirection))) (Math|Rotator|MakeRotator 0.0 0.0 (+ (.yaw (Utilities|Array|Get(aref) BGARotationArray _index)) (Utilities|Map|Find (Variables|Orientation|GetOrientationYawOffsets) (Variables|Orientation|GetCardinalDirection))))) 0 false 0))
        (bind _asbuilding_static_mesh (Utilities|Casting|CastToBuildingSMActor _returnvalue))
        (BuildingActor|SetPlayerPlacedforDestruction _asbuilding_static_mesh true)
        (Class|BuildingActor|SetTeam _asbuilding_static_mesh (Team|GetActorTeam (Actor|GetOwner)))
        (Building|InitializeKismetSpawnedBuildingActor _asbuilding_static_mesh 0 0 true 0 false)
        (bind _asbuilding_trap (Utilities|Casting|CastToBuildingTrap _asbuilding_static_mesh)
          (:then
            (Trap|SpawnedfromItemDefinition _asbuilding_trap "/Game/Athena/Items/Traps/DamageTrap/TID_Floor_Spikes_Athena_R_T01.TID_Floor_Spikes_Athena_R_T01" 1))
          (:CastFailed
            _))
        (if (not (and (Building|IsWorldLocValid (Building|GetBuildingStructuralSupportSystem) (+ (Transformation|GetActorLocation _asbuilding_static_mesh) "0,256.000000,960.000000")) (Building|IsWorldLocValid (Building|GetBuildingStructuralSupportSystem) (+ (Transformation|GetActorLocation _asbuilding_static_mesh) "0,256.000000,192.000000"))))
          (Actor|DestroyActor _asbuilding_static_mesh))))))

(fn HandleProps
  (params (Array :type Array%20of%20Actor%20Object%20References))
  (for _array_element Array
    (bind _asbuilding_prop (Utilities|Casting|CastToBuildingProp _array_element)
      (:then
        (bind _returnvalue (GameplayTags|HasMatchingGameplayTag(Message) _array_element "(TagName=\"Building.Type.Container.TreasureChest.Common\")")) ; Look for the gameplay tag this BP actor has before doing the potentally \nexpensive IsA check
        (if _returnvalue
          (Utilities|IsA(soft) _array_element "/Game/Building/ActorBlueprints/Containers/Tiered_Chest_Athena.Tiered_Chest_Athena_C"
            (:True
              _)
            (:False
              (if (Utilities|IsA _array_element "/Script/FortniteGame.BuildingPropBigHealth") :!label _label
                _
                (elif (Variables|/Script/FortniteGame.BuildingActor|GetbStructurallySupportOverlappingActors _asbuilding_prop)
                  _
                  (else
                    (bind _returnvalue_2 (CallFunction|CanDestroyObject _array_element :!label _label_4))
                    (if _returnvalue_2
                      (Utilities|Array|Add (Variables|Tower|GetBuildingActorsToDestroy) _array_element :!label _label_2)))))))
          (Utilities|Array|Add (Variables|Tower|GetBuildingActorsToIgnore) _array_element :!label _label_3))
        (goto _label))
      (:CastFailed
        (if (not (Utilities|IsA _array_element "/Game/Athena/SupplyDrops/BP_DamageBalloon_Athena.BP_DamageBalloon_Athena_C"))
          (bind _asbuilding_actor (Utilities|Casting|CastToBuildingActor _array_element))
          (if (GameplayTags|HasMatchingGameplayTag _asbuilding_actor "(TagName=\"Item.AppleSun\")")
            (goto _label_2)
            (else
              (goto _label_3))))))
    (goto _label_4)))

(fn SetBaseDestination
  (params (Hit :type Hit%20Result%20Structure%20%28by%20ref%29))
  (bind (_bblockinghit _binitialoverlap _time _distance _location _impactpoint _normal _impactnormal) (Collision|BreakHitResult Hit))
  (Networking|SwitchHasAuthority
    (:Authority
      (Variables|BuildingAndClearing|SetBaseDestination (Fortnite|SnaptoFortBuildingGrid (+ (Math|Vector|SelectVector (Math|Vector|MakeVector (.x (Variables|/Game/Athena/Items/Consumables/Parents/B_Prj_Comp_ClientPredicted_Data.B_Prj_Comp_ClientPredicted_Data_C|GetClientPredictedImpactLocation (Variables|Default|GetB_Prj_Comp_ClientPredicted_Data))) (.y (Variables|/Game/Athena/Items/Consumables/Parents/B_Prj_Comp_ClientPredicted_Data.B_Prj_Comp_ClientPredicted_Data_C|GetClientPredictedImpactLocation (Variables|Default|GetB_Prj_Comp_ClientPredicted_Data))) (.z _impactpoint)) _impactpoint (Utilities|Operators|Equal(==) (Variables|/Game/Athena/Items/Consumables/Parents/B_Prj_Comp_ClientPredicted_Data.B_Prj_Comp_ClientPredicted_Data_C|GetClientPredictedImpactLocation (Variables|Default|GetB_Prj_Comp_ClientPredicted_Data)) (Transformation|GetActorLocation) (Variables|Default|GetClientLocationErrorTolerance))) (* _impactnormal 1)) (Math|Vector|MakeVector 0.0 0.0 (Variables|SnaptoGrid|GetGridZOffset)) (Math|Vector|MakeVector 0.0 (Variables|SnaptoGrid|GetGridYOffset)))))))

(fn ShouldBounceOffActor
  (params (HitActor :type Actor%20Object%20Reference) (HitComponent :type Primitive%20Component%20Object%20Reference) (HitNormal :type Vector))
  (Utilities|IsValid HitActor
    (:Is%20Valid
      (Utilities|IsValid HitComponent
        (:Is%20Valid
          (if (or (Utilities|Array|ContainsItem (Variables|/Script/Engine.ActorComponent|GetComponentTags HitComponent) (Variables|Default|GetBounceTag)) (Utilities|Array|ContainsItem (Variables|Actor|GetTags HitActor) (Variables|Default|GetBounceTag)))
            (return true)
            (else
              (return (Math|Boolean|ORBoolean (Utilities|IsA HitActor "/Script/FortniteGame.FortPawn") (Utilities|IsA HitActor "/Script/FortniteGame.FortAthenaSupplyDrop") (Utilities|IsA HitActor "/Script/FortniteGame.FortAthenaSupplyDropBalloon") (Utilities|IsA HitActor "/Script/FortniteGame.FortProjectileBase") (Utilities|IsA HitActor "/Script/FortniteGame.BuildingContainer") (Utilities|IsA HitActor "/Script/FortniteGame.FortAthenaVehicle"))))))
        (:Is%20Not%20Valid
          _)))
    (:Is%20Not%20Valid
      _))
  (return false))

(fn SpawnBuildFX
  (params (Location :type Vector))
  (Effects|SpawnEmitteratLocation (Variables|Cosmetic|GetP_BuildBuilding) Location))

(fn SpawnTires
  (params (ReferenceLocation :type Vector))
  (bind _gridyoffset (Variables|SnaptoGrid|GetGridYOffset))
  (bind _gridzoffset (Variables|SnaptoGrid|GetGridZOffset))
  (bind _returnvalue (Fortnite|SnaptoFortBuildingGrid ReferenceLocation (Math|Vector|MakeVector 0.0 0.0 _gridzoffset) (Math|Vector|MakeVector 0.0 _gridyoffset)))
  (bind _cardinaldirection (Variables|Orientation|GetCardinalDirection))
  (bind _orientationyawoffsets (Variables|Orientation|GetOrientationYawOffsets))
  (bind _value (Utilities|Map|Find _orientationyawoffsets _cardinaldirection))
  (bind _cardinaldirection_2 (Variables|Orientation|GetCardinalDirection))
  (bind _orientationlocationoffsets (Variables|Orientation|GetOrientationLocationOffsets))
  (bind _value_2 (Utilities|Map|Find _orientationlocationoffsets _cardinaldirection_2))
  (bind _bgaclasses_tires (Variables|Tires|GetBGAClasses_Tires))
  (bind _bgalocations_tires (Variables|Tires|GetBGALocations_Tires))
  (bind _bgarotations_tires (Variables|Tires|GetBGARotations_Tires))
  (bind _bgaclasses_tires_2 (Variables|Tires|GetBGAClasses_Tires))
  (if (> (Utilities|Array|Length _bgaclasses_tires_2) 0)
    (if (Creative|Budget|CanPlaceClassInstances (Fortnite|GetFortPlayerControllerfromActor (Actor|GetOwner)) _bgaclasses_tires_2 ReferenceLocation)
      (for _index (range (Utilities|Array|Length (Variables|Tires|GetBGAClasses_Tires)))
        (bind (_outhit _returnvalue_2) (Collision|SphereTraceByChannel (+ (Math|Vector|vector+vector _returnvalue (Math|Vector|RotateVectorAroundAxis (Utilities|Array|Get(aref) _bgalocations_tires _index) _value "0, 0,1.000000") _value_2) "0, 0,384.000000") (Math|Vector|vector+vector _returnvalue (Math|Vector|RotateVectorAroundAxis (Utilities|Array|Get(aref) _bgalocations_tires _index) _value "0, 0,1.000000") _value_2) 32.0 "TraceTypeQuery10" false (Variables|Tower|GetBuildingActorsToIgnore) "None" true "(R=1.000000,G=0.000000,B=0.000000,A=1.000000)" "(R=0.000000,G=1.000000,B=0.000000,A=1.000000)" 15.0))
        (Development|PrintString (Utilities|String|Append "Tower Grenade Tire trace hit: " (Utilities|GetDisplayName (.HitActor (Collision|BreakHitResult _outhit)))) (Variables|Default|GetIsDebug) (Variables|Default|GetIsDebug))
        (bind _asbuilding_actor (Utilities|Casting|CastToBuildingActorClass (Utilities|Array|Get(acopy) _bgaclasses_tires _index))
          (:then
            (bind _returnvalue_3 (Building|SpawnBuildingActor _asbuilding_actor (Math|Transform|MakeTransform (select _returnvalue_2 (- (.Location (Collision|BreakHitResult _outhit)) "0, 0,32.000000") (Math|Vector|vector+vector _returnvalue (Math|Vector|RotateVectorAroundAxis (Utilities|Array|Get(aref) _bgalocations_tires _index) _value "0, 0,1.000000") _value_2)) (Math|Rotator|MakeRotator 0.0 0.0 (+ (.yaw (Utilities|Array|Get(aref) _bgarotations_tires _index)) _value)))))
            (bind _asbuilding_actor_2 (Utilities|Casting|CastToBuildingActor _returnvalue_3))
            (Building|InitializeKismetSpawnedBuildingActor _asbuilding_actor_2 0 0 true 0 false)
            (BuildingActor|SetPlayerPlacedforDestruction _asbuilding_actor_2 true)
            (if (not (and (Building|IsWorldLocValid (Building|GetBuildingStructuralSupportSystem) (+ (Transformation|GetActorLocation _returnvalue_3) "0,256.000000,960.000000")) (Building|IsWorldLocValid (Building|GetBuildingStructuralSupportSystem) (+ (Transformation|GetActorLocation _returnvalue_3) "0.000000,256.000000,192.000000"))))
              (Actor|DestroyActor _returnvalue_3)))
          (:CastFailed
            (Game|SpawnActorfromClass (Utilities|Array|Get(acopy) _bgaclasses_tires _index) (Math|Transform|MakeTransform (Fortnite|SnaptoFortBuildingGrid (.location (Math|Transform|MakeTransform (select _returnvalue_2 (- (.Location (Collision|BreakHitResult _outhit)) "0, 0,32.000000") (Math|Vector|vector+vector _returnvalue (Math|Vector|RotateVectorAroundAxis (Utilities|Array|Get(aref) _bgalocations_tires _index) _value "0, 0,1.000000") _value_2)) (Math|Rotator|MakeRotator 0.0 0.0 (+ (.yaw (Utilities|Array|Get(aref) _bgarotations_tires _index)) _value))))) (.rotation (Math|Transform|MakeTransform (select _returnvalue_2 (- (.Location (Collision|BreakHitResult _outhit)) "0, 0,32.000000") (Math|Vector|vector+vector _returnvalue (Math|Vector|RotateVectorAroundAxis (Utilities|Array|Get(aref) _bgalocations_tires _index) _value "0, 0,1.000000") _value_2)) (Math|Rotator|MakeRotator 0.0 0.0 (+ (.yaw (Utilities|Array|Get(aref) _bgarotations_tires _index)) _value)))) (.scale (Math|Transform|MakeTransform (select _returnvalue_2 (- (.Location (Collision|BreakHitResult _outhit)) "0, 0,32.000000") (Math|Vector|vector+vector _returnvalue (Math|Vector|RotateVectorAroundAxis (Utilities|Array|Get(aref) _bgalocations_tires _index) _value "0, 0,1.000000") _value_2)) (Math|Rotator|MakeRotator 0.0 0.0 (+ (.yaw (Utilities|Array|Get(aref) _bgarotations_tires _index)) _value))))) "AlwaysSpawn" "MultiplyWithRoot" 0 0)))))))

(fn ConstructionScript
  (Networking|CanExecuteCosmeticEvents
    (:True
      (Components|Movement|SetUpdatedComponent (Variables|Components|GetRotatingMovement) (Variables|Components|GetMesh)))))
